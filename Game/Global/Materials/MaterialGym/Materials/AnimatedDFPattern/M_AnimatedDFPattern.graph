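# Material Gym material for an animated distance-field pattern, using Super Levels T3 parameters and a Bengal Basher suit texture.
# Time minus the texture's green channel, scaled by speed, feeds a Frac that loops each region's phase.
# That phase moves a soft SphereMask threshold across the red channel and fades it in and out.
# The mask blends tint into pattern color, then a colored Fresnel rim is added.
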Material /Game/Global/Materials/MaterialGym/Materials/AnimatedDFPattern/M_AnimatedDFPattern
graph {
 properties {bAutomaticallySetUsageInEditor=false, pixelDepthOffsetMode=PDOM_Legacy, usedShadingModels=Default Lit}
 nodes [
  n1 = VectorParameter() {group=00_SuperLevels_T3, parameterName=SL_T3_SecondaryTintColor, defaultValue=<rgba 0.0117 0.0153 0.026 1>}
  n2 = VectorParameter() {group=00_SuperLevels_T3, parameterName=SL_T3_SecondaryPatternColor, defaultValue=<rgba 0.2127 0.3667 0.5938 1>}
  n3 = Time()
  n4 = TextureSample(UVs=Multiply(A=TextureCoordinate(), B=ScalarParameter() {defaultValue=4, group=00_SuperLevels_T3, parameterName=SL_T3_SecondaryUVScale})) {
   samplerSource=SSM_Wrap_WorldGroupSettings
   samplerType=SAMPLERTYPE_LinearColor
   texture=/Game/Characters/Player/Male/Large/Bodies/M_LRG_BengalBasher_Suit/Skins/SL/Textures/T_S28_SL3_DF
  }
  n5 = Frac(None=Multiply(A=ScalarParameter() {defaultValue=0.3, group=00_SuperLevels_T3, parameterName=SL_T3_SecondaryPatternSpeed}, B=Subtract(A=@n3, B=@n4.G)))
  n6 = SphereMask(A=@n5, B=Constant() {r=0.5}) {attenuationRadius=0.5, hardnessPercent=0}
  n7 = SphereMask(
   A=SphereMask(A=LinearInterpolate(Alpha=@n5) {constA=0.9, constB=0.65}, B=@n4.R) {attenuationRadius=0.125, hardnessPercent=0}
   B=Constant() {r=0.35}
  ) {attenuationRadius=0.3, hardnessPercent=0}
  Add(
   A=Multiply(
    A=Fresnel() {baseReflectFraction=0}
    B=VectorParameter() {group=00_SuperLevels_T3, parameterName=SL_FresnelColor, defaultValue=<rgba 0.6094 2.0248 3 1>}.RGB
   )
   B=LinearInterpolate(A=@n1.RGB, B=@n2.RGB, Alpha=Multiply(A=@n6, B=@n7))
  )
 ]
}
